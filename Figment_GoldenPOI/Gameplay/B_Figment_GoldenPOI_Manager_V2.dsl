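; Server-driven Golden POI manager that randomly selects which locations turn golden in a match.
; After playlist and navigation readiness plus a short delay, it draws distinct POI tags, with enablement and count set by scalable float rows.
; It replicates the tags once via a dormancy flush, broadcasts completion, and updates POI legend icons and minimap overrides.
; Spawners at chosen POIs, found at runtime or preset, begin spawning; the rest are destroyed.
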(event EventBeginPlay
  (if (Networking|IsServer)
    (Class|FortAsyncActionCurrentPlaylistReady|PlaylistDataReadyAsync
      (:OnReady
        (if (CallFunction|IsEnabled)
          (Default|CheckNavigationReady
            (:true
              (Utilities|FlowControl|Delay 2.0)
              (CallFunction|PickSpawners))))))))

(fn GetAllSpawners
  (if (Variables|Default|GetbFindSpawnersAtRuntime)
    ;[1 1 1]
    (bind _outactors (Actor|GetAllActorsOfClass (Variables|Default|GetSpawnerClass)))
    (for _array_element _outactors
      (bind _asb_athena_super_dingo_spawner_figment_golden_poi (Utilities|Casting|CastToB_Athena_SuperDingo_Spawner_Figment_GoldenPOI _array_element))
      (Utilities|Array|Add (LocalVariables|GetOutSpawners) _asb_athena_super_dingo_spawner_figment_golden_poi))
    (return (LocalVariables|GetOutSpawners))
    (else
      (return (Variables|Default|GetSpawners)))))

(fn GetNumToSpawn
  (return (Math|Random|RandomIntegerinRange (Math|Float|Ceil (ScalableFloat|GetValueatLevel (Variables|Default|GetRow_MinNumSpawn))) (Math|Float|Ceil (ScalableFloat|GetValueatLevel (Variables|Default|GetRow_MaxNumSpawn))))))

(fn IsEnabled
  (return (Class|AthenaFunctionLibrary|RowtoBool (Variables|Default|GetRow_Enabled))))

(fn IsLocationGolden
  (params (LocationTag :type Gameplay%20Tag%20Structure))
  (return (Utilities|Array|ContainsItem (Variables|Default|GetGoldenPOITags) LocationTag)))

(fn PickPOITags
  (LocalVariables|SetRemainingTags (Variables|Default|GetPOITags))
  (Utilities|FlowControl|ForLoopwithBreak 0 (- (CallFunction|GetNumtoSpawn) 1)
    (:LoopBody
      (if (Utilities|Array|IsNotEmpty (LocalVariables|GetRemainingTags))
        _
        (else
          _))
      (bind _output_get (LocalVariables|SetCurrentTag (Utilities|Array|RandomArrayItem (LocalVariables|GetRemainingTags))))
      (Development|PrintString (Utilities|String|Append (GameplayTags|GetDebugStringfromGameplayTag _output_get) " chosen to be golden") true true "(R=0.000000,G=0.660000,B=1.000000,A=1.000000)" 1e+01)
      (Utilities|Array|Add (LocalVariables|GetLocalOutTags) _output_get)
      (Utilities|Array|RemoveItem (LocalVariables|GetRemainingTags) (LocalVariables|GetCurrentTag)))
    (:Completed
      _))
  (return (LocalVariables|GetLocalOutTags)))

(fn PickSpawners
  (bind _outtags (CallFunction|PickPOITags))
  (bind _output_get (LocalVariables|SetSelectedPOITags _outtags))
  (if (Utilities|Array|IsNotEmpty _output_get)
    ;[1 1 1] Replicate selected tags down once
    (Variables|Default|SetGoldenPOITags (LocalVariables|GetSelectedPOITags))
    (Default|CallTagsFinishedBeingSet)
    (Variables|Default|SetTagsSet true)
    (Networking|FlushNetDormancy)
    (CallFunction|SetLegendaryPOIIcons)
    (CallFunction|SetMiniMapOverrideforSelectedPOIs (LocalVariables|GetSelectedPOITags))
    (bind _spawners (CallFunction|GetAllSpawners))
    (for _array_element _spawners
      (if (Utilities|Array|ContainsItem (LocalVariables|GetSelectedPOITags) (Variables|/Figment_GoldenPOI/Gameplay/B_Athena_SuperDingo_Spawner_Figment_GoldenPOI.B_Athena_SuperDingo_Spawner_Figment_GoldenPOI_C|GetPOITag _array_element))
        (Class|BAthenaSuperDingoSpawner|BeginSpawning _array_element)
        (else
          (Actor|DestroyActor _array_element))))))

(fn SetLegendaryPOIIcons
  (for _array_element (Class|FortPoiManager|GetAllPoiVolumes (GameState|GetPoiManager (Fortnite|Athena|GameState|GetGameStateAthena)))
    (for _array_element_2 (GameplayTags|BreakGameplayTagContainer (Class|FortPoiVolume|GetLocationTags _array_element :!class /Script/FortniteGame.FortPoiVolume))
      (bind _returnvalue (CallFunction|IsLocationGolden _array_element_2))
      (if _returnvalue
        (Class|FortPoiDiscoverableComponent|SetupPOIMapandLegendInfo (Actor|GetComponentbyClass _array_element "/Script/FortniteGame.FortPoi_DiscoverableComponent") true false)
        (Class|AthenaSpecialActorComponent|SetMapLegendTag (Actor|GetComponentbyClass _array_element "/Script/FortniteGame.FortPoi_DiscoverableComponent") (Variables|Default|GetLegendaryPOITag))
        (else
          (Class|FortPoiDiscoverableComponent|SetupPOIMapandLegendInfo (Actor|GetComponentbyClass _array_element "/Script/FortniteGame.FortPoi_DiscoverableComponent") false true))))))

(fn SetMiniMapOverrideForSelectedPOIs
  (params (SelectedPOITags :type Array%20of%20Gameplay%20Tag%20Structures))
  (LocalVariables|SetMiniMapOverrideComponent (Actor|GetComponentbyClass (Fortnite|Athena|GameState|GetGameStateAthena) "/Game/Blueprints/Components/GSC_MiniMapDataOverride.GSC_MiniMapDataOverride_C"))
  (if (and (Utilities|Array|IsNotEmpty SelectedPOITags) (Utilities|GameplayTagContainerNotEmpty (Variables|Default|GetPOITextContextTags)))
    (Class|FortGameStateComponentMiniMapDataOverride|OverrideLocations (LocalVariables|GetMiniMapOverrideComponent) (Variables|Default|GetPOITextContextTags) (GameplayTags|MakeGameplayTagContainerfromArray SelectedPOITags))))

(fn ConstructionScript)
